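# Sound cue for the looping daytime exterior wind ambience at the Chrome NothingTree POI, using outdoor ambient class and attenuation settings.
# A quality-level node feeds the quad wind loop to its first two slots and the stereo loop to the last two.
# Each branch chains two endlessly looping 60-second envelopers, one modulating volume and one pitch, so the looped wave slowly swells and drifts like gusting wind.
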SoundCue /Game/Sounds/Artemis/POI/Chrome_NothingTree/Ambience_Chrome_NothingTree_Exterior_Day_Wind_Cue
root=QualityLevel_2
nodes [
 {
  id=Enveloper_0
  children [Enveloper_4]
  properties {
   bLoop=true
   bLoopIndefinitely=true
   loopEnd=60
   pitchCurve {editorCurveData {keys [{time=0}]}}
   volumeCurve {editorCurveData {keys [
      <at 0 1 RCIM_Cubic>
      {arriveTangent=0, interpMode=RCIM_Cubic, leaveTangent=0, tangentMode=RCTM_User, time=15, value=0.5}
      {arriveTangent=0.0001, interpMode=RCIM_Cubic, leaveTangent=0.0001, tangentMode=RCTM_User, time=30, value=1}
      <at 45 0.7 RCIM_Cubic>
      <at 60 1 RCIM_Cubic>
     ]}}
  }
 }
 {
  id=Enveloper_4
  children [WavePlayer_7]
  properties {
   bLoop=true
   bLoopIndefinitely=true
   loopEnd=60
   pitchCurve {editorCurveData {keys [
      {arriveTangent=-0.0013, interpMode=RCIM_Cubic, time=0}
      {arriveTangent=-0.0007, interpMode=RCIM_Cubic, leaveTangent=-0.0007, time=16.7492, value=0.8782}
      <at 31.7492 0.9782 RCIM_Cubic>
      {arriveTangent=0.0008, interpMode=RCIM_Cubic, leaveTangent=0.0008, time=46.7492, value=0.8782}
      {time=60}
     ]}}
   volumeCurve {editorCurveData {keys [<at 0 1>]}}
  }
 }
 {
  id=Enveloper_5
  children [WavePlayer_6]
  properties {
   bLoop=true
   bLoopIndefinitely=true
   loopEnd=60
   pitchCurve {editorCurveData {keys [
      {arriveTangent=-0.0013, interpMode=RCIM_Cubic, time=0}
      {arriveTangent=-0.0006, interpMode=RCIM_Cubic, leaveTangent=-0.0006, time=15.6395, value=0.8812}
      <at 30.6395 0.9812 RCIM_Cubic>
      {arriveTangent=0.0006, interpMode=RCIM_Cubic, leaveTangent=0.0006, time=45.6395, value=0.8812}
      {time=60}
     ]}}
   volumeCurve {editorCurveData {keys [<at 0 1>]}}
  }
 }
 {
  id=Enveloper_6
  children [Enveloper_5]
  properties {
   bLoop=true
   bLoopIndefinitely=true
   loopEnd=60
   pitchCurve {editorCurveData {keys [{time=0}]}}
   volumeCurve {editorCurveData {keys [
      <at 0 1 RCIM_Cubic>
      {arriveTangent=0, interpMode=RCIM_Cubic, leaveTangent=0, tangentMode=RCTM_User, time=15, value=0.5}
      {arriveTangent=0.0001, interpMode=RCIM_Cubic, leaveTangent=0.0001, tangentMode=RCTM_User, time=30, value=1}
      <at 45 0.7 RCIM_Cubic>
      <at 60 1 RCIM_Cubic>
     ]}}
  }
 }
 {id=QualityLevel_2, children [Enveloper_6,Enveloper_6,Enveloper_0,Enveloper_0]}
 {
  id=WavePlayer_6
  properties {bLooping=true, soundWaveAssetPtr=&/Ambience_NothingTree_Desolate_Wind_Day_Quad}
 }
 {
  id=WavePlayer_7
  properties {bLooping=true, soundWaveAssetPtr=&/Ambience_NothingTree_Desolate_Wind_Day_Stereo}
 }
]
properties {
 attenuationSettings=&Fort_Ambient_Athena_Att
 maxDistance=2097152
 pitchMultiplier=0.9
 soundClassObject=&Fort_Ambient_Outdoors_Athena
 volumeMultiplier=0.376
}
