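# Looping outdoor ambience cue for the branches of the Chrome Nothing Tree POI, combining a steady branch loop with intermittent wood sounds.
# A quality-level node selects a three-layer mix for its first two slots and a simpler loop-plus-distant-stress mix for the last two.
# The full mix attenuates each layer separately, with creaks and distant stresses randomly picked, delayed and modulated.
# Concurrency is overridden to cap instances, with distance-based volume scaling.
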SoundCue /Game/Sounds/Artemis/POI/Chrome_NothingTree/Ambience_NothingTree_Branches_Loop_Cue
root=QualityLevel_1
nodes [
 {id=Attenuation_0, children [Modulator_0]}
 {id=Attenuation_1, children [Modulator_1]}
 {id=Attenuation_2, children [WavePlayer_14]}
 {id=Attenuation_3, children [Mixer_3]}
 {id=Delay_0, children [Random_2], properties {delayMax=12, delayMin=6}}
 {id=Delay_1, children [Random_1], properties {delayMax=12, delayMin=3}}
 {id=Delay_2, children [Random_0], properties {delayMax=12, delayMin=6}}
 {id=Looping_0, children [Delay_0]}
 {id=Looping_1, children [Delay_1]}
 {id=Looping_2, children [Delay_2]}
 {id=Mixer_2, children [Attenuation_0, Attenuation_1, Attenuation_2], properties {inputVolume [1,1.6,0.6]}}
 {id=Mixer_3, children [WavePlayer_15, Looping_1], properties {inputVolume [1,1]}}
 {id=Modulator_0, children [Looping_2], properties {pitchMax=1.1, pitchMin=0.9, volumeMin=0.8}}
 {id=Modulator_1, children [Looping_0], properties {pitchMax=1.1, pitchMin=0.9}}
 {id=QualityLevel_1, children [Mixer_2, Mixer_2, Attenuation_3, Attenuation_3]}
 {
  id=Random_0
  children [WavePlayer_7, WavePlayer_8, WavePlayer_9, WavePlayer_10, WavePlayer_11, WavePlayer_12, WavePlayer_13]
  properties {weights [1,1,1,1,1,1,1]}
 }
 {id=Random_1, children [WavePlayer_18, WavePlayer_28, WavePlayer_29], properties {weights [1,1,1]}}
 {
  id=Random_2
  children [WavePlayer_16, WavePlayer_17, WavePlayer_19, WavePlayer_20, WavePlayer_21, WavePlayer_22, WavePlayer_23, WavePlayer_24]
  properties {weights [1,1,1,1,1,1,1,1]}
 }
 {id=WavePlayer_10, properties {soundWaveAssetPtr=/Game/Sounds/Artemis/POI/Chrome_NothingTree/Spots/Spot_NothingTree_Wood_Stresses_Distant_03}}
 {id=WavePlayer_11, properties {soundWaveAssetPtr=/Game/Sounds/Artemis/POI/Chrome_NothingTree/Spots/Spot_NothingTree_Wood_Stresses_Distant_04}}
 {id=WavePlayer_12, properties {soundWaveAssetPtr=/Game/Sounds/Artemis/POI/Chrome_NothingTree/Spots/Spot_NothingTree_Wood_Stresses_Distant_05}}
 {id=WavePlayer_13, properties {soundWaveAssetPtr=/Game/Sounds/Artemis/POI/Chrome_NothingTree/Spots/Spot_NothingTree_Wood_Stresses_Distant_06}}
 {
  id=WavePlayer_14
  properties {bLooping=true, soundWaveAssetPtr=&/Ambience_NothingTree_Branch_Loop_Close}
 }
 {
  id=WavePlayer_15
  properties {bLooping=true, soundWaveAssetPtr=&Ambience_NothingTree_Branch_Loop_Close}
 }
 {id=WavePlayer_16, properties {soundWaveAssetPtr=/Game/Sounds/Artemis/POI/Chrome_NothingTree/Spots/Spot_Chrome_NothingTree_Branch_Creak_01}}
 {id=WavePlayer_17, properties {soundWaveAssetPtr=/Game/Sounds/Artemis/POI/Chrome_NothingTree/Spots/Spot_Chrome_NothingTree_Branch_Creak_02}}
 {id=WavePlayer_18, properties {soundWaveAssetPtr=/Game/Sounds/Artemis/POI/Chrome_NothingTree/Spots/Spot_NothingTree_Wood_Stresses_Distant_07}}
 {id=WavePlayer_19, properties {soundWaveAssetPtr=/Game/Sounds/Artemis/POI/Chrome_NothingTree/Spots/Spot_Chrome_NothingTree_Branch_Creak_03}}
 {id=WavePlayer_20, properties {soundWaveAssetPtr=/Game/Sounds/Artemis/POI/Chrome_NothingTree/Spots/Spot_Chrome_NothingTree_Branch_Creak_04}}
 {id=WavePlayer_21, properties {soundWaveAssetPtr=/Game/Sounds/Artemis/POI/Chrome_NothingTree/Spots/Spot_Chrome_NothingTree_Branch_Creak_05}}
 {id=WavePlayer_22, properties {soundWaveAssetPtr=/Game/Sounds/Artemis/POI/Chrome_NothingTree/Spots/Spot_Chrome_NothingTree_Branch_Creak_06}}
 {id=WavePlayer_23, properties {soundWaveAssetPtr=/Game/Sounds/Artemis/POI/Chrome_NothingTree/Spots/Spot_Chrome_NothingTree_Branch_Creak_07}}
 {id=WavePlayer_24, properties {soundWaveAssetPtr=/Game/Sounds/Artemis/POI/Chrome_NothingTree/Spots/Spot_Chrome_NothingTree_Branch_Creak_08}}
 {id=WavePlayer_28, properties {soundWaveAssetPtr=&Spot_NothingTree_Wood_Stresses_Distant_06}}
 {id=WavePlayer_29, properties {soundWaveAssetPtr=&Spot_NothingTree_Wood_Stresses_Distant_05}}
 {id=WavePlayer_7, properties {soundWaveAssetPtr=&Spot_NothingTree_Wood_Stresses_Distant_07}}
 {id=WavePlayer_8, properties {soundWaveAssetPtr=/Game/Sounds/Artemis/POI/Chrome_NothingTree/Spots/Spot_NothingTree_Wood_Stresses_Distant_01}}
 {id=WavePlayer_9, properties {soundWaveAssetPtr=/Game/Sounds/Artemis/POI/Chrome_NothingTree/Spots/Spot_NothingTree_Wood_Stresses_Distant_02}}
]
properties {
 attenuationSettings=/Game/Sounds/Artemis/POI/RealityTree/Spot_Reality_Tree_Folliage_Loop_Att
 bOverrideConcurrency=true
 maxDistance=2097152
 soundClassObject=&Fort_Ambient_Outdoors_Athena
 volumeMultiplier=0.451
 concurrencyOverrides {maxCount=2, volumeScaleMode=Distance}
}
