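# Outdoor ambient sound cue for the Chrome Nothing Tree POI, playing a level 3 Chrome biome night loop.
# A quality-level switch sends its first two slots to the quad recording and its last two to the stereo version.
# Each looping wave passes through nested, endlessly looping envelopers: a 30-second pitch curve inside a 60-second volume swell that varies the bed.
# The cue uses high-pass ambient attenuation with slightly lowered pitch and reduced volume.
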SoundCue /Game/Sounds/Artemis/POI/Chrome_NothingTree/Biome_Chrome_NothingTree_Night_Loop_Lvl3_Ext_Cue
root=QualityLevel_0
nodes [
 {
  id=Enveloper_0
  children [WavePlayer_1]
  properties {
   bLoop=true
   bLoopIndefinitely=true
   loopEnd=30
   pitchCurve {editorCurveData {keys [{arriveTangent=-0.0013, interpMode=RCIM_Cubic, time=0}, <at 15.9468 1.05 RCIM_Cubic>, {interpMode=RCIM_Cubic, leaveTangent=0.0014, time=30}]}}
   volumeCurve {editorCurveData {keys [<at 0 1>]}}
  }
 }
 {
  id=Enveloper_1
  children [WavePlayer_0]
  properties {
   bLoop=true
   bLoopIndefinitely=true
   loopEnd=30
   pitchCurve {editorCurveData {keys [{arriveTangent=-0.0013, interpMode=RCIM_Cubic, time=0}, <at 15.9468 1.05 RCIM_Cubic>, {interpMode=RCIM_Cubic, leaveTangent=0.0014, time=30}]}}
   volumeCurve {editorCurveData {keys [<at 0 1>]}}
  }
 }
 {
  id=Enveloper_2
  children [Enveloper_1]
  properties {
   bLoop=true
   bLoopIndefinitely=true
   loopEnd=60
   pitchCurve {editorCurveData {keys [{time=0}]}}
   volumeCurve {editorCurveData {keys [
      {arriveTangent=-0.01, interpMode=RCIM_Cubic, time=0, value=1}
      {arriveTangent=-0.005, interpMode=RCIM_Cubic, leaveTangent=-0.005, time=10, value=0.8}
      <at 25 1 RCIM_Cubic>
      {arriveTangent=0.0001, interpMode=RCIM_Cubic, leaveTangent=0.0001, tangentMode=RCTM_User, time=40, value=0.8}
      <at 60 1>
     ]}}
  }
 }
 {
  id=Enveloper_4
  children [Enveloper_0]
  properties {
   bLoop=true
   bLoopIndefinitely=true
   loopEnd=60
   pitchCurve {editorCurveData {keys [{time=0}]}}
   volumeCurve {editorCurveData {keys [
      {arriveTangent=-0.01, interpMode=RCIM_Cubic, time=0, value=1}
      <at 10 0.9 RCIM_Cubic>
      {arriveTangent=0.002, interpMode=RCIM_Cubic, leaveTangent=0.002, time=20, value=1}
      <at 40 0.8 RCIM_Cubic>
      <at 60 1>
     ]}}
  }
 }
 {id=QualityLevel_0, children [Enveloper_2,Enveloper_2,Enveloper_4,Enveloper_4]}
 {id=WavePlayer_0, properties {bLooping=true, soundWaveAssetPtr=/Game/Sounds/Artemis/Biomes/Chrome/AMB_Chrome_Night_Loop_Lvl3_02_Quad}}
 {id=WavePlayer_1, properties {bLooping=true, soundWaveAssetPtr=/Game/Sounds/Artemis/Biomes/Chrome/AMB_Chrome_Night_Loop_Lvl3_02_Stereo}}
]
properties {
 attenuationSettings=/Game/Sounds/Attenuation/Fort_Ambient_Athena_250Hz_HP_Att
 maxDistance=2097152
 pitchMultiplier=0.9
 soundClassObject=&Fort_Ambient_Outdoors_Athena
 volumeMultiplier=0.551
 preEffectBusSends [{}]
}
